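# Material function for item shop UI theming that outputs a gradient mask rising toward both the top and bottom edges.
# It remaps UVs around the center and takes the absolute vertical component to mirror the gradient.
# A Subtract offset and Falloff multiplier shape the band before it is saturated and scaled by Intensity.
# UVs, Subtract and Falloff inputs default to MF_UI_UVScaling's fill-scaled output and same-named scalar parameters.
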MaterialFunction /Game/UI/Frontend/ItemShop/Art/_MaterialFunctions_ShopTheming/MF_UI_TopBottomGradient
graph {nodes [
  n1 = MaterialFunctionCall() {desc=0, materialFunction=&MF_UI_UVScaling}
  n2 = Abs(None=ConstantBiasScale(None=FunctionInput(Preview=@n1.Scaled To Fill) {bUsePreviewValueAsDefault=true, inputName=UVs, inputType=FunctionInput_Vector2, sortPriority=1}) {bias=-0.5, scale=2})
  n3 = ScalarParameter() {defaultValue=0.37, parameterName=Subtract}
  n4 = Multiply(
   A=Subtract(
    A=ComponentMask(None=@n2) {g=true}
    B=FunctionInput(Preview=@n3) {bUsePreviewValueAsDefault=true, inputName=Subtract, inputType=FunctionInput_Scalar, sortPriority=2}
   )
   B=FunctionInput(Preview=ScalarParameter() {defaultValue=3.5, parameterName=Falloff}) {bUsePreviewValueAsDefault=true, inputName=Falloff, inputType=FunctionInput_Scalar, sortPriority=3}
  )
  FunctionOutput(None=Multiply(
    A=Saturate(None=@n4)
    B=FunctionInput() {bUsePreviewValueAsDefault=true, inputName=Intensity, inputType=FunctionInput_Scalar, sortPriority=4}
   ))
 ]}
